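# Sound cue for a wet foley sweetener on landing for the BandSoda large quadruped cosmetic companion.
# A quality-level node sends its first two slots to an equal-weight random pick of ten wet "Med" waves, and its last two to only the first five.
# It plays at reduced volume in the pets sound class, using companion foley attenuation and the shared quadruped footstep concurrency.
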SoundCue /CosmeticCompanions/Assets/Quadruped/BandSoda/Audio/FoleySweetener/SC_MImosa_Quadraped_Lrg_Wet_Add_Land
root=QualityLevel_1
nodes [
 {id=QualityLevel_1, children [Random_0,Random_0,Random_3,Random_3]}
 {
  id=Random_0
  children [WavePlayer_0,WavePlayer_1,WavePlayer_2,WavePlayer_3,WavePlayer_4,WavePlayer_5,WavePlayer_6,WavePlayer_7,WavePlayer_8,WavePlayer_9]
  properties {weights [1,1,1,1,1,1,1,1,1,1]}
 }
 {id=Random_3, children [WavePlayer_10, WavePlayer_11, WavePlayer_12, WavePlayer_13, WavePlayer_14], properties {weights [1,1,1,1,1]}}
 {
  id=WavePlayer_0
  properties {soundWaveAssetPtr=/CosmeticCompanions/Assets/Quadruped/BandSoda/Audio/FoleySweetener/wavs/Mimosa_Quadraped_Lrg_Wet_Add_Med_01}
 }
 {
  id=WavePlayer_1
  properties {soundWaveAssetPtr=/CosmeticCompanions/Assets/Quadruped/BandSoda/Audio/FoleySweetener/wavs/Mimosa_Quadraped_Lrg_Wet_Add_Med_02}
 }
 {
  id=WavePlayer_10
  properties {soundWaveAssetPtr=&Mimosa_Quadraped_Lrg_Wet_Add_Med_01}
 }
 {
  id=WavePlayer_11
  properties {soundWaveAssetPtr=&Mimosa_Quadraped_Lrg_Wet_Add_Med_02}
 }
 {
  id=WavePlayer_12
  properties {soundWaveAssetPtr=/CosmeticCompanions/Assets/Quadruped/BandSoda/Audio/FoleySweetener/wavs/Mimosa_Quadraped_Lrg_Wet_Add_Med_03}
 }
 {
  id=WavePlayer_13
  properties {soundWaveAssetPtr=/CosmeticCompanions/Assets/Quadruped/BandSoda/Audio/FoleySweetener/wavs/Mimosa_Quadraped_Lrg_Wet_Add_Med_04}
 }
 {
  id=WavePlayer_14
  properties {soundWaveAssetPtr=/CosmeticCompanions/Assets/Quadruped/BandSoda/Audio/FoleySweetener/wavs/Mimosa_Quadraped_Lrg_Wet_Add_Med_05}
 }
 {
  id=WavePlayer_2
  properties {soundWaveAssetPtr=&Mimosa_Quadraped_Lrg_Wet_Add_Med_03}
 }
 {
  id=WavePlayer_3
  properties {soundWaveAssetPtr=&Mimosa_Quadraped_Lrg_Wet_Add_Med_04}
 }
 {
  id=WavePlayer_4
  properties {soundWaveAssetPtr=&Mimosa_Quadraped_Lrg_Wet_Add_Med_05}
 }
 {
  id=WavePlayer_5
  properties {soundWaveAssetPtr=/CosmeticCompanions/Assets/Quadruped/BandSoda/Audio/FoleySweetener/wavs/Mimosa_Quadraped_Lrg_Wet_Add_Med_06}
 }
 {
  id=WavePlayer_6
  properties {soundWaveAssetPtr=/CosmeticCompanions/Assets/Quadruped/BandSoda/Audio/FoleySweetener/wavs/Mimosa_Quadraped_Lrg_Wet_Add_Med_07}
 }
 {
  id=WavePlayer_7
  properties {soundWaveAssetPtr=/CosmeticCompanions/Assets/Quadruped/BandSoda/Audio/FoleySweetener/wavs/Mimosa_Quadraped_Lrg_Wet_Add_Med_08}
 }
 {
  id=WavePlayer_8
  properties {soundWaveAssetPtr=/CosmeticCompanions/Assets/Quadruped/BandSoda/Audio/FoleySweetener/wavs/Mimosa_Quadraped_Lrg_Wet_Add_Med_09}
 }
 {
  id=WavePlayer_9
  properties {soundWaveAssetPtr=/CosmeticCompanions/Assets/Quadruped/BandSoda/Audio/FoleySweetener/wavs/Mimosa_Quadraped_Lrg_Wet_Add_Med_10}
 }
]
properties {
 attenuationSettings=/Game/Sounds/Attenuation/AS_Companion_Foley
 maxDistance=1200
 priority=0
 soundClassObject=&Fort_Creatures_Pets
 volumeMultiplier=0.4
 concurrencySet [/CosmeticCompanions/Assets/Audio_Shared/Foley/Quadraped/CONC_Mimosa_Quadraped_Footstep]
}
