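# Sound cue for the wet "slide short" foley sweetener of the BandSoda large quadruped cosmetic companion.
# A quality-level node routes its first two slots to a random pick of ten "Wet_Add_Long" waves and its last two to the first four of those waves.
# It plays quietly in the pets sound class, using companion foley attenuation and the shared quadruped footstep concurrency.
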SoundCue /CosmeticCompanions/Assets/Quadruped/BandSoda/Audio/FoleySweetener/SC_MImosa_Quadraped_Lrg_Wet_Add_SlideShort
root=QualityLevel_0
nodes [
 {id=QualityLevel_0, children [Random_2,Random_2,Random_4,Random_4]}
 {
  id=Random_2
  children [
   WavePlayer_12
   WavePlayer_13
   WavePlayer_14
   WavePlayer_15
   WavePlayer_16
   WavePlayer_18
   WavePlayer_19
   WavePlayer_20
   WavePlayer_21
   WavePlayer_22
  ]
  properties {weights [1,1,1,1,1,1,1,1,1,1]}
 }
 {id=Random_4, children [WavePlayer_23, WavePlayer_24, WavePlayer_25, WavePlayer_26], properties {weights [1,1,1,1]}}
 {
  id=WavePlayer_12
  properties {soundWaveAssetPtr=/CosmeticCompanions/Assets/Quadruped/BandSoda/Audio/FoleySweetener/wavs/Mimosa_Quadraped_Lrg_Wet_Add_Long_01}
 }
 {
  id=WavePlayer_13
  properties {soundWaveAssetPtr=/CosmeticCompanions/Assets/Quadruped/BandSoda/Audio/FoleySweetener/wavs/Mimosa_Quadraped_Lrg_Wet_Add_Long_02}
 }
 {
  id=WavePlayer_14
  properties {soundWaveAssetPtr=/CosmeticCompanions/Assets/Quadruped/BandSoda/Audio/FoleySweetener/wavs/Mimosa_Quadraped_Lrg_Wet_Add_Long_03}
 }
 {
  id=WavePlayer_15
  properties {soundWaveAssetPtr=/CosmeticCompanions/Assets/Quadruped/BandSoda/Audio/FoleySweetener/wavs/Mimosa_Quadraped_Lrg_Wet_Add_Long_04}
 }
 {
  id=WavePlayer_16
  properties {soundWaveAssetPtr=/CosmeticCompanions/Assets/Quadruped/BandSoda/Audio/FoleySweetener/wavs/Mimosa_Quadraped_Lrg_Wet_Add_Long_05}
 }
 {
  id=WavePlayer_18
  properties {soundWaveAssetPtr=/CosmeticCompanions/Assets/Quadruped/BandSoda/Audio/FoleySweetener/wavs/Mimosa_Quadraped_Lrg_Wet_Add_Long_06}
 }
 {
  id=WavePlayer_19
  properties {soundWaveAssetPtr=/CosmeticCompanions/Assets/Quadruped/BandSoda/Audio/FoleySweetener/wavs/Mimosa_Quadraped_Lrg_Wet_Add_Long_07}
 }
 {
  id=WavePlayer_20
  properties {soundWaveAssetPtr=/CosmeticCompanions/Assets/Quadruped/BandSoda/Audio/FoleySweetener/wavs/Mimosa_Quadraped_Lrg_Wet_Add_Long_08}
 }
 {
  id=WavePlayer_21
  properties {soundWaveAssetPtr=/CosmeticCompanions/Assets/Quadruped/BandSoda/Audio/FoleySweetener/wavs/Mimosa_Quadraped_Lrg_Wet_Add_Long_09}
 }
 {
  id=WavePlayer_22
  properties {soundWaveAssetPtr=/CosmeticCompanions/Assets/Quadruped/BandSoda/Audio/FoleySweetener/wavs/Mimosa_Quadraped_Lrg_Wet_Add_Long_10}
 }
 {
  id=WavePlayer_23
  properties {soundWaveAssetPtr=&Mimosa_Quadraped_Lrg_Wet_Add_Long_01}
 }
 {
  id=WavePlayer_24
  properties {soundWaveAssetPtr=&Mimosa_Quadraped_Lrg_Wet_Add_Long_02}
 }
 {
  id=WavePlayer_25
  properties {soundWaveAssetPtr=&Mimosa_Quadraped_Lrg_Wet_Add_Long_03}
 }
 {
  id=WavePlayer_26
  properties {soundWaveAssetPtr=&Mimosa_Quadraped_Lrg_Wet_Add_Long_04}
 }
]
properties {
 attenuationSettings=/Game/Sounds/Attenuation/AS_Companion_Foley
 maxDistance=1200
 priority=0
 soundClassObject=&Fort_Creatures_Pets
 volumeMultiplier=0.2
 concurrencySet [/CosmeticCompanions/Assets/Audio_Shared/Foley/Quadraped/CONC_Mimosa_Quadraped_Footstep]
}
